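; Movement mode extension logic for cube surfing, driving the camera, audio, FX, tags and events while a pawn surfs.
; Activation caches controller, weapon, audio and camera, records the weapon for durability, applies camera modes and modifiers, and binds surfing dispatchers.
; Handlers toggle weapon FX, play client-only movement audio, tag falling, and broadcast boost, jump and crash events through the global event router.
; Landing audio waits on a timer; deactivation unbinds and cleans up.
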(event EventOnExtensionActivate
  (params (RuntimeContext :type MMERuntime%20Context%20Structure%20%28by%20ref%29))
  (bind _ownerpawn (Variables|FortMovementMode_BaseExtLogic|GetOwnerPawn))
  ;[0.01376 0.2962 1] Activate -> Bind events
  (|CallParentFunction RuntimeContext)
  (Default|CacheCommonVariables)
  (CallFunction|SetUpforDurability RuntimeContext)
  (Default|AddCameraMode)
  (CallFunction|ToggleEventBindings true)
  (Utilities|OwnerPawnNetRole
    (:Client
      (MovementLoops|OnMovementChange (Variables|Default|GetSpatialTravelAudComp) true))
    (:Server
      _))
  ;[1 0 0.06004] Apply camera modifier while surfing
  (bind _returnvalue (Fortnite|Camera|AddCameraModifierforOwningPlayerController "/CubeSurfing/Gameplay/Camera/CameraModifier_CubeSurfing_Surfing.CameraModifier_CubeSurfing_Surfing_C" (Variables|FortMovementMode_BaseExtLogic|GetOwnerPawn)))
  (Variables|Default|SetCameraModifierHandle (Utilities|Casting|CastToCameraModifier_CubeSurfing_Surfing _returnvalue))
  ;[1 0 0.06004] Jump camera motion
  (bind _asynctaskproxy (Ability|Tasks|WaitGameplayTagAddToActor _ownerpawn "(TagName=\"Granted.CubeSurfing.PreBoost\")")
    (:then
      (Variables|Default|SetWaitPreBoostTask _asynctaskproxy)
      (Utilities|Time|SetTimerbyEvent (delegate Custom|AllowLand) (Variables|Default|GetLandAudioDelayTime)))
    (:Added
      (MovementOneshots|OnPreBoost (Variables|Default|GetSpatialTravelAudComp))
      (bind _returnvalue_2 (Fortnite|Camera|AddCameraModifierforOwningPlayerController "/CubeSurfing/Gameplay/Camera/CameraModifier_CubeSurfing_Boost.CameraModifier_CubeSurfing_Boost_C" _ownerpawn))
      (Variables|Default|SetBoostModifierHandle _returnvalue_2))))

(event Custom|AllowLand
  ;[1 0 0.06004] Jump camera motion
  (Variables|Default|SetAllowLandAudio true))

(event Custom|OnAirborne
  ;[0.497 0.02466 0.9412 0.5] ON AIR / LAND
  (Ability|GameplayEffect|AddLooseGameplayTags (Variables|FortMovementMode_BaseExtLogic|GetOwnerPawn) (GameplayTags|MakeLiteralGameplayTagContainer "(GameplayTags=((TagName=\"MovementMode.Falling\")))") true)
  (Utilities|OwnerPawnNetRole
    (:Client
      (MovementLoops|OnAirborne (Variables|Default|GetSpatialTravelAudComp) false)
      (CallFunction|SetonLandFX false))))

(event Custom|OnLanded
  ;[0.497 0.02466 0.9412 0.5] ON AIR / LAND
  (Ability|GameplayEffect|RemoveLooseGameplayTags (Variables|FortMovementMode_BaseExtLogic|GetOwnerPawn) (GameplayTags|MakeLiteralGameplayTagContainer "(GameplayTags=((TagName=\"MovementMode.Falling\")))") true)
  (Utilities|OwnerPawnNetRole
    (:Client
      (Variables|Default|GetAllowLandAudio
        (:then
          (MovementLoops|OnLand (Variables|Default|GetSpatialTravelAudComp))
          (CallFunction|SetonLandFX true))
        (:else)))))

(event EventOnExtensionDeactivate
  (params (RuntimeContext :type MMERuntime%20Context%20Structure%20%28by%20ref%29))
  (bind _spatialtravelaudcomp (Variables|Default|GetSpatialTravelAudComp))
  ;[1 0.07441 0.1059] Deactivate -> Unbind events and end FX
  (Utilities|FlowControl|Sequence
    (:then_0
      (Default|RemoveCameraMode)
      (CallFunction|ToggleEventBindings false)
      (Utilities|IsValid _spatialtravelaudcomp
        (:Is%20Valid
          (MovementLoops|OnMovementChange _spatialtravelaudcomp false))
        (:Is%20Not%20Valid
          _))
      (Utilities|Time|ClearandInvalidateTimerbyHandle (Variables|FortMovementMode_BaseExtLogic|GetOwnerPawn) (Variables|Default|GetCameraModifierTimer))
      (|CallParentFunction RuntimeContext)
      (CallFunction|ResetCameraMod)
      (CallFunction|OnCubeSurfingBoostEndEvent 0.0)
      (CallFunction|SetBoostingFX false)
      (CallFunction|SetWaterFX false)
      (CallFunction|SetonLandFX false)
      (Variables|Default|SetAllowLandAudio false)
      (MovementLoops|OnDismount (Variables|Default|GetSpatialTravelAudComp)))
    (:then_1
      (Ability|SendGameplayEventtoActor (Variables|FortMovementMode_BaseExtLogic|GetOwnerPawn) "(TagName=\"Event.CubeSurfing.Stop\")"))))

(event Custom|BoostStarted
  ;[0.497 0.02466 0.9412 0.5] ON AIR / LAND
  (CallFunction|AddBoostCue))

(event Custom|OnCubeSurfingBoost_Event
  (params (Max%20Duration :type Float%20%28double-precision%29))
  ;[0.01376 0.2962 1] Boosting
  (Utilities|FlowControl|Sequence
    (:then_0
      (CallFunction|PlayBoostCamera))
    (:then_1
      (Ability|SendGameplayEventtoActor (Variables|FortMovementMode_BaseExtLogic|GetOwnerPawn) "(TagName=\"Event.CubeSurfing.Boost\")")
      (CallFunction|SetBoostingFX true))
    (:then_2
      (GameplayEventRouter|BroadcastEvent (GameplayEventRouter|GetGlobalEventRouter (Variables|FortMovementMode_BaseExtLogic|GetOwnerPawn)) (Utilities|Struct|MakeCubeSurfingEventBoostDuration Max%20Duration) (Variables|FortMovementMode_BaseExtLogic|GetOwnerPawn)))))

(event Custom|ResetCameraMod
  (bind _cachedplayercontroller (Variables|Default|GetCachedPlayerController))
  ;[0.01376 0.2962 1] InputAction Boost (unhooked for now)
  (Utilities|IsValid _cachedplayercontroller
    (:Is%20Valid
      (bind _asfort_player_camera_base (Utilities|Casting|CastToFortPlayerCameraBase (Variables|/Script/Engine.PlayerController|GetPlayerCameraManager _cachedplayercontroller))
        (:then
          (Game|Player|RemoveCameraModifier _asfort_player_camera_base (Variables|Default|GetBoostModifierHandle)))
        (:CastFailed
          _))
      (Ability|Async|EndAction (Variables|Default|GetWaitPreBoostTask)))
    (:Is%20Not%20Valid
      _)))

(event Custom|OnCubeSurfingWaterActorChanged_Event
  (params (OldWaterBody :type Water%20Body%20Object%20Reference) (NewWaterBody :type Water%20Body%20Object%20Reference))
  (bind _returnvalue (Utilities|IsValid NewWaterBody :!class /Script/Engine.KismetSystemLibrary))
  ;[0.01376 0.2962 1] Anim
  (Ability|SendGameplayEventtoActor (Variables|FortMovementMode_BaseExtLogic|GetOwnerPawn) "(TagName=\"Event.CubeSurfing.WaterActorChanged\")" (Utilities|Struct|MakeGameplayEventData "(TagName=\"Event.CubeSurfing.WaterActorChanged\")" 0 0 OldWaterBody NewWaterBody))
  (Utilities|OwnerPawnNetRole
    (:Client
      (MovementLoops|OnWatermovement (Variables|Default|GetSpatialTravelAudComp) _returnvalue)
      (CallFunction|SetWaterFX _returnvalue))))

(event Custom|OnCubeSurfingCrashLanding_Event
  ;[0.01376 0.2962 1] Anim
  (GameplayEventRouter|BroadcastEvent (GameplayEventRouter|GetGlobalEventRouter (Variables|FortMovementMode_BaseExtLogic|GetOwnerPawn)) (Utilities|Struct|MakeCubeSurfingEventCrashLanding) (Variables|FortMovementMode_BaseExtLogic|GetOwnerPawn)))

(event Custom|OnCubeSurfingBoostEnd_Event
  (params (Cooldown :type Float%20%28double-precision%29))
  ;[0.01376 0.2962 1] Boosting
  (Utilities|FlowControl|Sequence
    (:then_0
      (CallFunction|SetBoostingFX false)
      (Camera|RemoveCameraModeOverride (Variables|Default|GetCameraBase) (Variables|Default|GetCameraMode_BoostTag) self true))
    (:then_1
      (GameplayEventRouter|BroadcastEvent (GameplayEventRouter|GetGlobalEventRouter (Variables|FortMovementMode_BaseExtLogic|GetOwnerPawn)) (Utilities|Struct|MakeCubeSurfingEventBoostCooldown Cooldown) (Variables|FortMovementMode_BaseExtLogic|GetOwnerPawn)))))

(event Custom|OnCubeSurfingJump_Event
  ;[0.01376 0.2962 1] Anim
  (GameplayEventRouter|BroadcastEvent (GameplayEventRouter|GetGlobalEventRouter (Variables|FortMovementMode_BaseExtLogic|GetOwnerPawn)) (Utilities|Struct|MakeCubeSurfingEventJump) (Variables|FortMovementMode_BaseExtLogic|GetOwnerPawn))
  (Utilities|OwnerPawnNetRole
    (:Client
      (MovementLoops|OnJump (Variables|Default|GetSpatialTravelAudComp)))))

(event Custom|OnCubeSurfingFailedBoost_Event
  ;[0.01376 0.2962 1] Boosting
  (GameplayEventRouter|BroadcastEvent (GameplayEventRouter|GetGlobalEventRouter (Variables|FortMovementMode_BaseExtLogic|GetOwnerPawn)) (Utilities|Struct|MakeCubeSurfingEventBoostFailed) (Variables|FortMovementMode_BaseExtLogic|GetOwnerPawn)))

(fn AddBoostCue
  (Projectile|AddGameplayCue%28Looping%29Local (Variables|FortMovementMode_BaseExtLogic|GetOwnerPawn) (Variables|Default|GetGC_ChargeStart) (Ability|GameplayCue|MakeGameplayCueParameters 0.0)))

(fn AddCameraMode
  (bind _camera (Default|GetPlayerCamera))
  (Utilities|IsValid _camera
    (:Is%20Valid
      (Camera|AddCameraModeOverride _camera (Variables|Default|GetCameraMode_Base) (Variables|Default|GetCameraMode_BaseTag) 0 (Variables|FortMovementMode_BaseExtLogic|GetOwnerPawn) 0))))

(fn CacheCommonVariables
  (bind _ownerpawn (Variables|FortMovementMode_BaseExtLogic|GetOwnerPawn))
  (bind _asfort_player_controller (Utilities|Casting|CastToFortPlayerController (Pawn|GetController _ownerpawn))
    (:then
      (bind _output_get (Variables|Default|SetCachedPlayerController _asfort_player_controller)))
    (:CastFailed
      _))
  (bind _asb_cube_surfing_weapon (Utilities|Casting|CastToB_CubeSurfing_Weapon (Weapon|GetCurrentWeapon _ownerpawn)))
  (bind _output_get_2 (Variables|Default|SetCubeWeapon _asb_cube_surfing_weapon))
  (Variables|Default|SetSpatialTravelAudComp (Actor|GetComponentbyClass _output_get_2 "/Game/Sounds/Fort_Audio_Blueprints/SpatialTravelAudioComponent/SpatialTravel_SoundLibraryComponent.SpatialTravel_SoundLibraryComponent_C"))
  (bind _asfort_player_camera_base (Utilities|Casting|CastToFortPlayerCameraBase (Variables|/Script/Engine.PlayerController|GetPlayerCameraManager _output_get)))
  (Variables|Default|SetCameraBase _asfort_player_camera_base))

(fn CreatePawnComponent)

(fn GetCubeSurfingCameraMode
  (bind _camera (Default|GetPlayerCamera))
  (Utilities|IsValid _camera
    (:Is%20Valid
      (bind _ascamera_mode_cube_surfing (Utilities|Casting|CastToCameraMode_CubeSurfing (Class|FortPlayerCameraBase|GetCurrentCameraMode _camera)))
      (return _ascamera_mode_cube_surfing))))

(fn GetPlayerCamera
  (bind _asfort_player_camera_base (Utilities|Casting|CastToFortPlayerCameraBase (Variables|/Script/Engine.PlayerController|GetPlayerCameraManager (Variables|Default|GetCachedPlayerController))))
  (return _asfort_player_camera_base))

(fn PlayBoostCamera
  (Camera|AddCameraModeOverride (Variables|Default|GetCameraBase) (Variables|Default|GetCameraMode_Boost) (Variables|Default|GetCameraMode_BoostTag) 0 self (Variables|FortMovementMode_BaseExtLogic|GetOwnerPawn)))

(fn RemoveBoostCue
  (Projectile|RemoveGameplayCue%28Looping%29Local (Variables|FortMovementMode_BaseExtLogic|GetOwnerPawn) (Variables|Default|GetGC_ChargeStart) (Ability|GameplayCue|MakeGameplayCueParameters 0.0)))

(fn RemoveCameraMode
  (bind _cachedplayercontroller (Variables|Default|GetCachedPlayerController))
  (Utilities|FlowControl|Sequence
    (:then_0
      (Utilities|IsValid _cachedplayercontroller
        (:Is%20Valid
          (bind _asfort_player_camera_base (Utilities|Casting|CastToFortPlayerCameraBase (Variables|/Script/Engine.PlayerController|GetPlayerCameraManager _cachedplayercontroller)))
          (Camera|RemoveCameraModeOverride _asfort_player_camera_base (Variables|Default|GetCameraMode_BaseTag) (Variables|FortMovementMode_BaseExtLogic|GetOwnerPawn) false))))
    (:then_1
      (Class|FortCameraModifierTargetRotationFrustum|RemoveSelffromCameraManager (Variables|Default|GetCameraModifierHandle)))))

(fn SetBoostingFX
  (params (Enabled? :type Boolean))
  (bind _cubeweapon (Variables|Default|GetCubeWeapon))
  (Utilities|IsValid _cubeweapon
    (:Is%20Valid
      (Class|BCubeSurfingWeapon|ToggleIsBoosting _cubeweapon Enabled?))))

(fn SetInputMappingContextEnabled
  (params (Enabled? :type Boolean))
  (bind _cachedplayercontroller (Variables|Default|GetCachedPlayerController))
  (bind _returnvalue (PlayerController|LocalPlayerSubsystems|GetEnhancedInputLocalPlayerSubsystem _cachedplayercontroller))
  (bind _cubesurfinginputcontext (LocalVariables|GetCubeSurfingInputContext))
  (Utilities|IsValid _cachedplayercontroller
    (:Is%20Valid
      (if Enabled?
        (Input|AddMappingContext _returnvalue _cubesurfinginputcontext 2000 "(bIgnoreAllPressedKeysUntilRelease=True,bForceImmediately=False,bNotifyUserSettings=False)")
        (else
          (Input|RemoveMappingContext _returnvalue _cubesurfinginputcontext "(bIgnoreAllPressedKeysUntilRelease=True,bForceImmediately=False,bNotifyUserSettings=False)"))))))

(fn SetOnLandFX
  (params (bIsOnLand :type Boolean))
  (bind _cubeweapon (Variables|Default|GetCubeWeapon))
  (Utilities|IsValid _cubeweapon
    (:Is%20Valid
      (Class|BCubeSurfingWeapon|ToggleIsonLand _cubeweapon bIsOnLand))))

(fn SetUpForDurability
  (params (Runtime%20Context :type MMERuntime%20Context%20Structure))
  (bind _ownerpawn (Variables|FortMovementMode_BaseExtLogic|GetOwnerPawn))
  (bind _returnvalue (Weapon|GetCurrentWeapon _ownerpawn))
  (bind _returnvalue_2 (Class|FortMovementModeExtBlueprintLibrary|GetRuntimeDatafromContext Runtime%20Context "/Script/CubeSurfingRuntime.FortMovementMode_RuntimeDataCubeSurfing"))
  (if (Utilities|IsA _returnvalue "/CubeSurfing/Gameplay/Actors/B_CubeSurfing_Weapon.B_CubeSurfing_Weapon_C")
    (Class|FortMovementModeItemActivatedRuntimeData|SetActivatingItemfromWeapon _returnvalue_2 _returnvalue)
    (else
      (Fortnite|FortShippingLog _ownerpawn "Cube Surfing MME activated, but the Current Weapon is not B_Cube_Surfing_Weapon!  Cannot set activating item from weapon reliably." true))))

(fn SetWaterFX
  (params (bIsOnWater :type Boolean))
  (bind _cubeweapon (Variables|Default|GetCubeWeapon))
  (Utilities|IsValid _cubeweapon
    (:Is%20Valid
      (Class|BCubeSurfingWeapon|ToggleIsonWater _cubeweapon bIsOnWater))))

(fn StartCameraBurst)

(fn ToggleEventBindings
  (params (Enable :type Boolean))
  (if Enable
    (EventDispatchers|BindEventtoOnCubeSurfingBoost (EventDispatchers|CreateEvent :!event OnCubeSurfingBoost_Event))
    (EventDispatchers|BindEventtoOnCubeSurfingBoostEnd (EventDispatchers|CreateEvent :!event OnCubeSurfingBoostEnd_Event))
    (EventDispatchers|BindEventtoOnCubeSurfingAirborne (EventDispatchers|CreateEvent :!event OnAirborne))
    (EventDispatchers|BindEventtoOnCubeSurfingLanded (EventDispatchers|CreateEvent :!event OnLanded))
    (EventDispatchers|BindEventtoOnCubeSurfingBoostCharging (EventDispatchers|CreateEvent :!event BoostStarted))
    (EventDispatchers|BindEventtoOnCubeSurfingFailedBoost (EventDispatchers|CreateEvent :!event OnCubeSurfingFailedBoost_Event))
    (EventDispatchers|BindEventtoOnWaterActorChanged (EventDispatchers|CreateEvent :!event OnCubeSurfingWaterActorChanged_Event))
    (EventDispatchers|BindEventtoOnCrashLanding (EventDispatchers|CreateEvent :!event OnCubeSurfingCrashLanding_Event))
    (EventDispatchers|BindEventtoOnCubeSurfingJump (EventDispatchers|CreateEvent :!event OnCubeSurfingJump_Event))
    (else
      (EventDispatchers|UnbindallEventsfromOnCubeSurfingBoost)
      (EventDispatchers|UnbindallEventsfromOnCubeSurfingBoostEnd)
      (EventDispatchers|UnbindallEventsfromOnCubeSurfingAirborne)
      (EventDispatchers|UnbindallEventsfromOnCubeSurfingLanded)
      (EventDispatchers|UnbindallEventsfromOnCubeSurfingBoostCharging)
      (EventDispatchers|UnbindallEventsfromOnCubeSurfingFailedBoost)
      (EventDispatchers|UnbindallEventsfromOnWaterActorChanged)
      (EventDispatchers|UnbindallEventsfromOnCrashLanding)
      (EventDispatchers|UnbindallEventsfromOnCubeSurfingJump))))
